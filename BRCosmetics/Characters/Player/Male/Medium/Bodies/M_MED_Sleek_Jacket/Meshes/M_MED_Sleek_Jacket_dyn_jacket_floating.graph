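# Physics asset for the dynamic lower jacket on the male medium Sleek Jacket body.
# Six five-segment capsule chains (front, side, back, left and right) hang from master bodies linked to the pelvis, each joint swinging mainly on one axis with tight twist.
# Pelvis, thigh, calf and foot capsules give the lower-body shape.
# A few angularly free constraints cross-link neighbouring chains at segments 3 and 5, asymmetrically between sides.
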PhysicsAsset /BRCosmetics/Characters/Player/Male/Medium/Bodies/M_MED_Sleek_Jacket/Meshes/M_MED_Sleek_Jacket_dyn_jacket_floating
bodies {
 calf_l [{shapeType=Capsule, length=16.75, radius=14.125, center=<-12.6074 -2.8177 1.871>, rotation=<rot 89.4447 177.1457 0.0004>}]
 calf_r [{shapeType=Capsule, length=16.75, radius=14.125, center=<12.6074 2.8177 -1.871>, rotation=<rot 89.4447 177.1457 -179.9996>}]
 calf_twist_01_l [{shapeType=Capsule, length=19.5, radius=8.5, rotation=<rot 90 3.5763 -171.4235>}]
 calf_twist_01_r [{shapeType=Capsule, length=19.5, radius=8.5, rotation=<rot 90 -3.5763 1.4238>}]
 dyn_jacketBack_01_l [{shapeType=Capsule, length=24.375, radius=10, center=<4 0 7>, rotation=<rot 0 0 95>}]
 dyn_jacketBack_01_r [{shapeType=Capsule, length=24.5, radius=9.875, center=<4 0 7>, rotation=<rot 0 0 79.9995>}]
 dyn_jacketBack_02_l [{shapeType=Capsule, length=25.375, radius=10, center=<6 0 7>, rotation=<rot 0 0 95>}]
 dyn_jacketBack_02_r [{shapeType=Capsule, length=25.5, radius=9.875, center=<6 0 7>, rotation=<rot 0 0 79.9994>}]
 dyn_jacketBack_03_l [{shapeType=Capsule, length=24.625, radius=12.875, center=<6 0 7>, rotation=<rot 0 0 95>}]
 dyn_jacketBack_03_r [{shapeType=Capsule, length=24.75, radius=12.75, center=<6 0 7>, rotation=<rot 0 0 79.9995>}]
 dyn_jacketBack_04_l [{shapeType=Capsule, length=25.625, radius=12.875, center=<8 0 7>, rotation=<rot 0 0 95>}]
 dyn_jacketBack_04_r [{shapeType=Capsule, length=25.75, radius=12.75, center=<8 0 7>, rotation=<rot 0 0 79.9994>}]
 dyn_jacketBack_05_l [{shapeType=Capsule, length=25.625, radius=12.875, center=<6 0 7>, rotation=<rot 0 0 95>}]
 dyn_jacketBack_05_r [{shapeType=Capsule, length=25.25, radius=13.25, center=<6 0 7>, rotation=<rot 0 0 79.9994>}]
 dyn_jacketBack_mstr_l [{shapeType=Capsule, length=0.505, radius=0.505, rotation=<rot 0 -90 0>}]
 dyn_jacketBack_mstr_r [{shapeType=Capsule, length=0.505, radius=0.505, rotation=<rot 0 -90 0>}]
 dyn_jacketFront_01_l [{shapeType=Capsule, length=14.375, radius=11.5, center=<6 -5 9>, rotation=<rot 0 0 95>}]
 dyn_jacketFront_01_r [{shapeType=Capsule, length=15, radius=10.25, center=<6 5 9>, rotation=<rot 0 0 90.0002>}]
 dyn_jacketFront_02_l [{shapeType=Capsule, length=14.375, radius=10.25, center=<6 -5 9>, rotation=<rot 0 0 95>}]
 dyn_jacketFront_02_r [{shapeType=Capsule, length=15, radius=10.125, center=<6 5 9>, rotation=<rot 0 0 90.0002>}]
 dyn_jacketFront_03_l [{shapeType=Capsule, length=13.125, radius=12.25, center=<6 -5 9>, rotation=<rot 0 0 95>}]
 dyn_jacketFront_03_r [{shapeType=Capsule, length=13.75, radius=12, center=<6 5 9>, rotation=<rot 0 0 90.0002>}]
 dyn_jacketFront_04_l [{shapeType=Capsule, length=13.125, radius=12.375, center=<8 -5 9>, rotation=<rot 0 0 95>}]
 dyn_jacketFront_04_r [{shapeType=Capsule, length=13.75, radius=12.375, center=<8 5 9>, rotation=<rot 0 0 90.0002>}]
 dyn_jacketFront_05_l [{shapeType=Capsule, length=13.125, radius=12.875, center=<6 -5 9>, rotation=<rot 0 0 95>}]
 dyn_jacketFront_05_r [{shapeType=Capsule, length=13.75, radius=13.125, center=<6 5 9>, rotation=<rot 0.4352 4.9819 90.0196>}]
 dyn_jacketFront_mstr_l [{shapeType=Capsule, length=0.505, radius=0.505, rotation=<rot 0 -90 0>}]
 dyn_jacketFront_mstr_r [{shapeType=Capsule, length=0.505, radius=0.505, rotation=<rot 0 -90 0>}]
 dyn_jacketSide_01_l [{shapeType=Capsule, length=27.125, radius=9.125, center=<4 -2 9>, rotation=<rot 0 0 85>}]
 dyn_jacketSide_01_r [{shapeType=Capsule, length=27.5, radius=8.75, center=<4 2 9>, rotation=<rot 0 0 90.0011>}]
 dyn_jacketSide_02_l [{shapeType=Capsule, length=27.875, radius=9.375, center=<6 -2 9>, rotation=<rot 0 0 85>}]
 dyn_jacketSide_02_r [{shapeType=Capsule, length=28, radius=9.25, center=<6 2 9>, rotation=<rot 0 0 90.0011>}]
 dyn_jacketSide_03_l [{shapeType=Capsule, length=27, radius=12.375, center=<6 -2 9>, rotation=<rot 0 0 85>}]
 dyn_jacketSide_03_r [{shapeType=Capsule, length=27.75, radius=11.625, center=<6 2 9>, rotation=<rot 0 0 90.0011>}]
 dyn_jacketSide_04_l [{shapeType=Capsule, length=28.375, radius=12, center=<8 -2 9>, rotation=<rot 0 0 85>}]
 dyn_jacketSide_04_r [{shapeType=Capsule, length=28.75, radius=11.625, center=<8 2 9>, rotation=<rot 0 0 90.0011>}]
 dyn_jacketSide_05_l [{shapeType=Capsule, length=28.25, radius=12.125, center=<6 -2 9>, rotation=<rot 0 0 85>}]
 dyn_jacketSide_05_r [{shapeType=Capsule, length=28.25, radius=12.125, center=<6 2 9>, rotation=<rot 0 0 90.0011>}]
 dyn_jacketSide_mstr_l [{shapeType=Capsule, length=0.505, radius=0.505, rotation=<rot 0 -90 0>}]
 dyn_jacketSide_mstr_r [{shapeType=Capsule, length=0.505, radius=0.505, rotation=<rot 0 -90 0>}]
 foot_l [{shapeType=Capsule, length=15.5, radius=12.125, center=<-3.5001 8.3295 0.0009>, rotation=<rot 0 0 -90>}]
 foot_r [{shapeType=Capsule, length=15.5, radius=12.125, center=<3.5001 -8.3295 -0.0009>, rotation=<rot 0 0 90>}]
 pelvis [{shapeType=Capsule, length=10.25, radius=10.25, center=<-2 0 0>}]
 thigh_l [{shapeType=Capsule, length=26.625, radius=10.375, center=<-20.2214 0.0124 0.9539>, rotation=<rot 88.5789 -171.6228 -173.8271>}]
 thigh_r [{shapeType=Capsule, length=30, radius=10.75, center=<20.2214 -0.0124 -0.9539>, rotation=<rot 88.5787 -171.6237 8.8461>}]
}
constraints [
 {bone1Name=dyn_jacketSide_01_l, bone2Name=dyn_jacketSide_mstr_l, swing1LimitDegrees=2, swing2LimitDegrees=30, twistLimitDegrees=2}
 {bone1Name=dyn_jacketSide_02_l, bone2Name=dyn_jacketSide_01_l, swing1LimitDegrees=2, swing2LimitDegrees=30, twistLimitDegrees=2}
 {bone1Name=dyn_jacketSide_03_l, bone2Name=dyn_jacketSide_02_l, swing1LimitDegrees=2, swing2LimitDegrees=30, twistLimitDegrees=2}
 {bone1Name=dyn_jacketSide_04_l, bone2Name=dyn_jacketSide_03_l, swing1LimitDegrees=2, swing2LimitDegrees=30, twistLimitDegrees=2}
 {bone1Name=dyn_jacketBack_01_l, bone2Name=dyn_jacketBack_mstr_l, swing1LimitDegrees=2, swing2LimitDegrees=30, twistLimitDegrees=2}
 {bone1Name=dyn_jacketBack_02_l, bone2Name=dyn_jacketBack_01_l, swing1LimitDegrees=2, swing2LimitDegrees=30, twistLimitDegrees=2}
 {bone1Name=dyn_jacketBack_03_l, bone2Name=dyn_jacketBack_02_l, swing1LimitDegrees=2, swing2LimitDegrees=30, twistLimitDegrees=2}
 {bone1Name=dyn_jacketBack_04_l, bone2Name=dyn_jacketBack_03_l, swing1LimitDegrees=2, swing2LimitDegrees=30, twistLimitDegrees=2}
 {bone1Name=dyn_jacketBack_01_r, bone2Name=dyn_jacketBack_mstr_r, swing1LimitDegrees=2, swing2LimitDegrees=30, twistLimitDegrees=2}
 {bone1Name=dyn_jacketBack_02_r, bone2Name=dyn_jacketBack_01_r, swing1LimitDegrees=2, swing2LimitDegrees=30, twistLimitDegrees=2}
 {bone1Name=dyn_jacketBack_03_r, bone2Name=dyn_jacketBack_02_r, swing1LimitDegrees=2, swing2LimitDegrees=30, twistLimitDegrees=2}
 {bone1Name=dyn_jacketBack_04_r, bone2Name=dyn_jacketBack_03_r, swing1LimitDegrees=2, swing2LimitDegrees=30, twistLimitDegrees=2}
 {bone1Name=dyn_jacketSide_01_r, bone2Name=dyn_jacketSide_mstr_r, swing1LimitDegrees=2, swing2LimitDegrees=30, twistLimitDegrees=2}
 {bone1Name=dyn_jacketSide_02_r, bone2Name=dyn_jacketSide_01_r, swing1LimitDegrees=2, swing2LimitDegrees=30, twistLimitDegrees=2}
 {bone1Name=dyn_jacketSide_03_r, bone2Name=dyn_jacketSide_02_r, swing1LimitDegrees=2, swing2LimitDegrees=30, twistLimitDegrees=2}
 {bone1Name=dyn_jacketSide_04_r, bone2Name=dyn_jacketSide_03_r, swing1LimitDegrees=2, swing2LimitDegrees=30, twistLimitDegrees=2}
 {
  bone1Name=dyn_jacketSide_mstr_l
  bone2Name=pelvis
  swing1LimitDegrees=45
  swing1Motion=Free
  swing2LimitDegrees=45
  swing2Motion=Free
  twistLimitDegrees=45
  twistMotion=Free
 }
 {
  bone1Name=dyn_jacketBack_mstr_l
  bone2Name=pelvis
  swing1LimitDegrees=45
  swing1Motion=Free
  swing2LimitDegrees=45
  swing2Motion=Free
  twistLimitDegrees=45
  twistMotion=Free
 }
 {
  bone1Name=dyn_jacketBack_mstr_r
  bone2Name=pelvis
  swing1LimitDegrees=45
  swing1Motion=Free
  swing2LimitDegrees=45
  swing2Motion=Free
  twistLimitDegrees=45
  twistMotion=Free
 }
 {
  bone1Name=dyn_jacketSide_mstr_r
  bone2Name=pelvis
  swing1LimitDegrees=45
  swing1Motion=Free
  swing2LimitDegrees=45
  swing2Motion=Free
  twistLimitDegrees=45
  twistMotion=Free
 }
 {
  bone1Name=dyn_jacketFront_mstr_l
  bone2Name=pelvis
  swing1LimitDegrees=45
  swing1Motion=Free
  swing2LimitDegrees=45
  swing2Motion=Free
  twistLimitDegrees=45
  twistMotion=Free
 }
 {bone1Name=dyn_jacketFront_01_l, bone2Name=dyn_jacketFront_mstr_l, swing1LimitDegrees=2, swing2LimitDegrees=30, twistLimitDegrees=2}
 {bone1Name=dyn_jacketFront_02_l, bone2Name=dyn_jacketFront_01_l, swing1LimitDegrees=2, swing2LimitDegrees=30, twistLimitDegrees=2}
 {bone1Name=dyn_jacketFront_03_l, bone2Name=dyn_jacketFront_02_l, swing1LimitDegrees=2, swing2LimitDegrees=30, twistLimitDegrees=2}
 {bone1Name=dyn_jacketFront_04_l, bone2Name=dyn_jacketFront_03_l, swing1LimitDegrees=2, swing2LimitDegrees=30, twistLimitDegrees=2}
 {
  bone1Name=dyn_jacketFront_mstr_r
  bone2Name=pelvis
  swing1LimitDegrees=45
  swing1Motion=Free
  swing2LimitDegrees=45
  swing2Motion=Free
  twistLimitDegrees=45
  twistMotion=Free
 }
 {bone1Name=dyn_jacketFront_01_r, bone2Name=dyn_jacketFront_mstr_r, swing1LimitDegrees=2, swing2LimitDegrees=30, twistLimitDegrees=2}
 {bone1Name=dyn_jacketFront_02_r, bone2Name=dyn_jacketFront_01_r, swing1LimitDegrees=2, swing2LimitDegrees=30, twistLimitDegrees=2}
 {bone1Name=dyn_jacketFront_03_r, bone2Name=dyn_jacketFront_02_r, swing1LimitDegrees=2, swing2LimitDegrees=30, twistLimitDegrees=2}
 {bone1Name=dyn_jacketFront_04_r, bone2Name=dyn_jacketFront_03_r, swing1LimitDegrees=2, swing2LimitDegrees=30, twistLimitDegrees=2}
 {
  bone1Name=foot_r
  bone2Name=calf_r
  swing1LimitDegrees=45
  swing1Motion=Free
  swing2LimitDegrees=45
  swing2Motion=Free
  twistLimitDegrees=45
  twistMotion=Free
 }
 {bone1Name=calf_r, bone2Name=thigh_r, swing1LimitDegrees=10, swing2LimitDegrees=45, twistLimitDegrees=2}
 {
  bone1Name=thigh_r
  bone2Name=pelvis
  swing1LimitDegrees=45
  swing1Motion=Free
  swing2LimitDegrees=45
  swing2Motion=Free
  twistLimitDegrees=45
  twistMotion=Free
 }
 {bone1Name=dyn_jacketSide_05_l, bone2Name=dyn_jacketSide_04_l, swing1LimitDegrees=2, swing2LimitDegrees=30, twistLimitDegrees=2}
 {bone1Name=dyn_jacketBack_05_l, bone2Name=dyn_jacketBack_04_l, swing1LimitDegrees=2, swing2LimitDegrees=30, twistLimitDegrees=2}
 {bone1Name=dyn_jacketFront_05_l, bone2Name=dyn_jacketFront_04_l, swing1LimitDegrees=2, swing2LimitDegrees=30, twistLimitDegrees=2}
 {bone1Name=dyn_jacketSide_05_r, bone2Name=dyn_jacketSide_04_r, swing1LimitDegrees=2, swing2LimitDegrees=30, twistLimitDegrees=2}
 {bone1Name=dyn_jacketBack_05_r, bone2Name=dyn_jacketBack_04_r, swing1LimitDegrees=2, swing2LimitDegrees=30, twistLimitDegrees=2}
 {bone1Name=dyn_jacketFront_05_r, bone2Name=dyn_jacketFront_04_r, swing1LimitDegrees=2, swing2LimitDegrees=30, twistLimitDegrees=2}
 {
  bone1Name=dyn_jacketSide_05_r
  bone2Name=dyn_jacketBack_05_r
  swing1LimitDegrees=45
  swing1Motion=Free
  swing2LimitDegrees=45
  swing2Motion=Free
  twistLimitDegrees=45
  twistMotion=Free
 }
 {
  bone1Name=foot_l
  bone2Name=calf_l
  swing1LimitDegrees=45
  swing1Motion=Free
  swing2LimitDegrees=45
  swing2Motion=Free
  twistLimitDegrees=45
  twistMotion=Free
 }
 {bone1Name=thigh_l, bone2Name=pelvis, swing1LimitDegrees=5, swing2LimitDegrees=60, twistLimitDegrees=2}
 {
  bone1Name=dyn_jacketFront_05_l
  bone2Name=dyn_jacketSide_05_l
  swing1LimitDegrees=45
  swing1Motion=Free
  swing2LimitDegrees=45
  swing2Motion=Free
  twistLimitDegrees=45
  twistMotion=Free
 }
 {
  bone1Name=dyn_jacketSide_03_r
  bone2Name=dyn_jacketFront_03_r
  swing1LimitDegrees=45
  swing1Motion=Free
  swing2LimitDegrees=45
  swing2Motion=Free
  twistLimitDegrees=45
  twistMotion=Free
 }
 {
  bone1Name=calf_twist_01_r
  bone2Name=calf_r
  swing1LimitDegrees=45
  swing1Motion=Free
  swing2LimitDegrees=45
  swing2Motion=Free
  twistLimitDegrees=45
  twistMotion=Free
 }
 {
  bone1Name=calf_twist_01_l
  bone2Name=calf_l
  swing1LimitDegrees=45
  swing1Motion=Free
  swing2LimitDegrees=45
  swing2Motion=Free
  twistLimitDegrees=45
  twistMotion=Free
 }
 {bone1Name=calf_l, bone2Name=thigh_l, swing1LimitDegrees=10, swing2LimitDegrees=45, twistLimitDegrees=2}
 {
  bone1Name=dyn_jacketSide_05_l
  bone2Name=dyn_jacketBack_05_l
  swing1LimitDegrees=45
  swing1Motion=Free
  swing2LimitDegrees=45
  swing2Motion=Free
  twistLimitDegrees=45
  twistMotion=Free
 }
]
